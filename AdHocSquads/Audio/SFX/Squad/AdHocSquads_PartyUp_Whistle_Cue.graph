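# Party-up whistle sound cue for AdHocSquads, in the in-game SFX class with overridden logarithmic distance attenuation.
# A mixer layers one of three equally weighted random whistle waves with a half-volume, randomly pitched-up troll appear sound.
# A fixed delay feeds the same random whistle node into the mixer again, producing a second whistle about a second after the first.
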SoundCue /AdHocSquads/Audio/SFX/Squad/AdHocSquads_PartyUp_Whistle_Cue
root=Mixer_0
nodes [
 {id=Delay_0, children [Random_0], properties {delayMax=1.1, delayMin=1.1}}
 {id=Mixer_0, children [Random_0,Modulator_6,Delay_0], properties {inputVolume [1,0.5,1]}}
 {id=Modulator_6, children [WavePlayer_84], properties {pitchMax=1.5, pitchMin=1.2}}
 {id=Random_0, children [WavePlayer_0,WavePlayer_1,WavePlayer_2], properties {weights [1,1,1]}}
 {id=WavePlayer_0, properties {soundWaveAssetPtr=&/AdHocSquads_PartyUp_Whistle_01}}
 {id=WavePlayer_1, properties {soundWaveAssetPtr=&/AdHocSquads_PartyUp_Whistle_03}}
 {id=WavePlayer_2, properties {soundWaveAssetPtr=&/AdHocSquads_PartyUp_Whistle_02}}
 {id=WavePlayer_84, properties {soundWaveAssetPtr=&/troll_appear_03}}
]
properties {
 bOverrideAttenuation=true
 maxDistance=810
 soundClassObject=&Fort_SFX_InGame
 volumeMultiplier=0.3
 attenuationOverrides {distanceAlgorithm=Logarithmic, falloffDistance=800, reverbDistanceMax=1000, reverbDistanceMin=10, attenuationShapeExtents {x=10}}
}
